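# NitroDrome checkpoint hit sound cue that plays one wave differently depending on whether it is triggered locally.
# A Branch on the IsLocal bool sends true to the wave through an Attenuation node with no settings asset assigned.
# False and unset share a quieter Mixer path that plays the same wave through the checkpoint hit attenuation asset.
# The cue is routed to the Fort_Traps sound class.
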SoundCue /NitroDrome/Assets/Audio/Checkpoints/NitroDrome_Checkpoint_Hit_Cue
root=Branch_0
nodes [
 {id=Attenuation_0, children [WavePlayer_0]}
 {
  id=Attenuation_1
  children [WavePlayer_1]
  properties {attenuationSettings=&/NitroDrome_Checkpoint_Hit_Att}
 }
 {id=Branch_0, children [Attenuation_0, Mixer_0, Mixer_0], properties {boolParameterName=IsLocal}}
 {id=Mixer_0, children [Attenuation_1], properties {inputVolume [0.7]}}
 {id=WavePlayer_0, properties {soundWaveAssetPtr=&/NitroDrome_Checkpoint_Hit_01}}
 {id=WavePlayer_1, properties {soundWaveAssetPtr=&NitroDrome_Checkpoint_Hit_01}}
]
properties {maxDistance=2097152, soundClassObject=&Fort_Traps, volumeMultiplier=0.85}
