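# Sound cue for the start of a checkpoint timer in NitroDrome.
# A single wave player feeds the checkpoint timer start sound straight to the output, with no randomization or mixing and a slight volume boost.
# It borrows the PA announcer voice-over attenuation asset for spatial falloff and is assigned to the Fort_Traps sound class.
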SoundCue /NitroDrome/Assets/Audio/Checkpoints/NitroDrome_Checkpoint_Timer_Start_Cue
root=WavePlayer_2
nodes [{id=WavePlayer_2, properties {soundWaveAssetPtr=&/NitroDrome_Checkpoint_Timer_Start}}]
properties {
 attenuationSettings=/NitroDrome/Assets/Audio/PA/VO/VO_PA_Announcer_Att
 maxDistance=20000
 soundClassObject=&Fort_Traps
 volumeMultiplier=1.2
}
